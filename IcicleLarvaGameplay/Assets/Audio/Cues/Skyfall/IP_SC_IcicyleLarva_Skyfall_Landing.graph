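# Sound cue for the landing of the Icicle Larva's Skyfall, reusing the Deimos ice spawner explosion sound wave.
# A single wave player feeds a modulator at the root, with the cue volume lowered.
# It plays through the large explosives sound class with the Skyfall incoming attenuation and a long maximum distance.
# A concurrency rule limits it to one instance per owner, stopping the oldest.
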SoundCue /IcicleLarvaGameplay/Assets/Audio/Cues/Skyfall/IP_SC_IcicyleLarva_Skyfall_Landing
root=Modulator_0
nodes [
 {id=Modulator_0, children [WavePlayer_2]}
 {id=WavePlayer_2, properties {soundWaveAssetPtr=/Game/Sounds/Fort_Monster_Deimos/Spawner_Ice/DeimosSpawner_Ice_Explode}}
]
properties {
 attenuationSettings=/IcicleLarvaGameplay/Assets/Audio/Settings/ATT_IcicleLarva_Skyfall_Incomming
 maxDistance=77000
 soundClassObject=/Game/Sounds/FortSoundClassesAndModes/Fort_Explosives_Large_PSM
 volumeMultiplier=0.284
 concurrencySet [/Game/Sounds/Concurrency_Rules/OneMaximum_Owner_StopOldest]
}
