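# Sound cue for the Icicle Larva skyfall, layering incoming and ice sounds audible over a very long range.
# A mixer blends a one-shot incoming wave, a looping incoming wave pitched down to half, and a quieter ice loop.
# Both incoming layers share one attenuation, while the ice loop has its own.
# The cue uses an important mission sound class and overrides concurrency to cap simultaneous instances.
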SoundCue /IcicleLarvaGameplay/Assets/Audio/Cues/Skyfall/IP_SC_IcicyleLarva_Skyfall_Loop
root=Mixer_0
nodes [
 {
  id=Attenuation_0
  children [WavePlayer_0]
  properties {attenuationSettings=/IcicleLarvaGameplay/Assets/Audio/Settings/ATT_IcicleLarva_Skyfall_Ice_Loop}
 }
 {
  id=Attenuation_1
  children [WavePlayer_2]
  properties {attenuationSettings=/IcicleLarvaGameplay/Assets/Audio/Settings/ATT_IcicleLarva_Skyfall_Incomming}
 }
 {
  id=Attenuation_2
  children [Modulator_0]
  properties {attenuationSettings=&ATT_IcicleLarva_Skyfall_Incomming}
 }
 {id=Mixer_0, children [Attenuation_1, Attenuation_2, Attenuation_0], properties {inputVolume [1,1,0.5]}}
 {id=Modulator_0, children [WavePlayer_4], properties {pitchMax=0.5, pitchMin=0.5}}
 {id=WavePlayer_0, properties {bLooping=true, soundWaveAssetPtr=/IcicleLarvaGameplay/Assets/Audio/Waves/IP_IcicleLarva_Skydrop_Ice_Loop}}
 {id=WavePlayer_2, properties {soundWaveAssetPtr=/IcicleLarvaGameplay/Assets/Audio/Waves/IP_IcicleLarva_Skydrop_Incomming}}
 {id=WavePlayer_4, properties {bLooping=true, soundWaveAssetPtr=/IcicleLarvaGameplay/Assets/Audio/Waves/IP_IcicleLarva_Skydrop_Incomming_Loop}}
]
properties {
 bOverrideConcurrency=true
 maxDistance=77000
 priority=30
 soundClassObject=&Fort_Mission_Ingame_Important_PSM
 volumeMultiplier=0.4
 concurrencyOverrides {maxCount=2}
}
