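# Sound cue for the start of the Icicle Larva Skyfall event, built from meteor and distant burst waves.
# A mixer layers a sonic boom and a distant end-burst, each with random pitch variation, plus a quieter meteor sweetener delayed by a fixed quarter second.
# It plays as an important in-game mission sound with long-range custom attenuation and a concurrency override capping it at two instances.
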SoundCue /IcicleLarvaGameplay/Assets/Audio/Cues/Skyfall/IP_SC_IcicyleLarva_Skyfall_Start
root=Mixer_0
nodes [
 {id=Delay_0, children [WavePlayer_5], properties {delayMax=0.25, delayMin=0.25}}
 {id=Mixer_0, children [Modulator_0,Modulator_1,Delay_0], properties {inputVolume [1,1,0.2]}}
 {id=Modulator_0, children [WavePlayer_3], properties {pitchMax=1.2, pitchMin=0.9}}
 {id=Modulator_1, children [WavePlayer_4], properties {pitchMax=1.2, pitchMin=0.9}}
 {id=WavePlayer_3, properties {soundWaveAssetPtr=/IcicleLarvaGameplay/Assets/Audio/Waves/IP_IcicleLarva_Meteor_Sonic_Boom_02}}
 {id=WavePlayer_4, properties {soundWaveAssetPtr=/IcicleLarvaGameplay/Assets/Audio/Waves/IP_IcicleLarva_Transform_EndBurst_Distant_01}}
 {id=WavePlayer_5, properties {soundWaveAssetPtr=/IcicleLarvaGameplay/Assets/Audio/Waves/IP_IcicleLarva_Meteor_5_1_Sweetener}}
]
properties {
 attenuationSettings=/IcicleLarvaGameplay/Assets/Audio/Settings/ATT_IcicleLarva_Skyfall_Start
 bOverrideConcurrency=true
 maxDistance=77000
 priority=30
 soundClassObject=&Fort_Mission_Ingame_Important_PSM
 volumeMultiplier=0.8
 concurrencyOverrides {maxCount=2}
}
